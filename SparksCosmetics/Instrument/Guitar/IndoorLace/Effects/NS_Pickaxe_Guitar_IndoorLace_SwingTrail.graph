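# Niagara swing-trail effect for the IndoorLace guitar pickaxe, layering three ribbons with sparkles.
# Two short-lived AnimTrail ribbons taper and fade, while a wisp ribbon sampled from the user skeletal mesh widens and breaks up with late curl noise.
# Flickering sparkles use spawn groups to switch size, color and glow within one emitter, inheriting capped velocity before drifting on curl noise.
# Ribbons take a time-of-day tint and dim in front-end views.
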NiagaraSystem /SparksCosmetics/Instrument/Guitar/IndoorLace/Effects/NS_Pickaxe_Guitar_IndoorLace_SwingTrail
dependencies {
 usedDataInterfaces [&NiagaraDataInterfaceCurve, &NiagaraDataInterfaceColorCurve, &NiagaraDataInterfaceSocketReader]
 usedDynamicInputs []
 usedModules [
  &self:SystemSpawnScript.SystemScriptSource.SystemScriptGraph.NiagaraNodeAssignment_3.SetVariables_#1
  &self:SystemSpawnScript.SystemScriptSource.SystemScriptGraph.NiagaraNodeAssignment_2.SetVariables_#2
  &SystemState
  &self:SystemSpawnScript.SystemScriptSource.SystemScriptGraph.NiagaraNodeAssignment_0.SetVariables_#3
  &self:SystemSpawnScript.SystemScriptSource.SystemScriptGraph.NiagaraNodeAssignment_1.SetVariables_#4
 ]
}
emitters {
 Flares_Sparkles {
  emitterUpdateScript [
   &EmitterState(Life Cycle Mode=::NewEnumerator1, Loop Behavior=::NewEnumerator1, Loop Duration=0.5, UseLoopDelay=false)
   &SpawnRate(
    SpawnRate=&RandomRangeFloat(Minimum=80, Maximum=85, Evaluation Type=::NewEnumerator1)
    Use Spawn Probability=false
   )
   &SpawnRate(
    SpawnGroup=1
    SpawnRate=&RandomRangeFloat(Minimum=20, Maximum=25, Evaluation Type=::NewEnumerator1)
    Use Spawn Probability=false
   ) {module=SpawnRate001}
  ]
  particleSpawnScript [
   /SetVariables(Particles.VisibilityTag=&SpawnGroup()) {isSetParameters=true}
   &InitializeParticle(
    Color=&MergeHelper_ColorSwitch(
     0=&RandomRangeLinearColor(Minimum=<rgba 1 0.7676 0.4189 0.5>, Maximum=<rgba 1.185 1.25 0.5892 1>)
     1=&RandomRangeLinearColor(Minimum=<rgba 4 3.0703 1.6758 1>, Maximum=<rgba 4.7402 5 2.3568 1>)
     2=?
     3=?
     Index=&SpawnGroup()
     Num=::NewEnumerator0
    )
    Lifetime Max=0.45
    Lifetime Min=0.3
    Lifetime Mode=::NewEnumerator1
    Sprite Size Mode=::NewEnumerator3
    Uniform Sprite Size=&MergeHelper_FloatSwitch(
     0=&RandomRangeFloat(Minimum=0.4, Maximum=0.75)
     1=&RandomRangeFloat(Minimum=8, Maximum=12)
     2=?
     3=?
     Index=&SpawnGroup()
     Num=::NewEnumerator0
    )
    UsePositionOffset=false
    Write Material Random=true
   )
   &CameraOffset(Camera Offset Amount=10)
   &SkeletalMeshLocation(MaskSpawnGroup=false, Mesh Sampling Type=::NewEnumerator5, Skeletal Mesh=$User.SkeletalMesh, Socket Sampling Mode=::NewEnumerator5)
   &ShapeLocation(
    Cylinder Height=35
    Cylinder Height Midpoint=0
    Cylinder Radius=3
    MaskSpawnGroup=false
    Override Seed=false
    Shape Primitive=::NewEnumerator1
    Surface Only=false
    UseOwnerScale=true
   )
   &InheritVelocity(
    Inherited Velocity Speed Limit:NiagaraFloat=&RandomRangeFloat(Minimum=50, Maximum=75)
    Speed Limit=true
    UseSpeedThreshold=false
   )
   &AddVelocity(Velocity=<0 300 0>) {enabled=false}
  ]
  particleUpdateScript [
   &ParticleState()
   /SetVariables(Particles.FlickerWaveform:NiagaraFloat=&Waveform(
     Global Phase=&Add_Float(A=$Engine.Time, B=&RandomRangeFloat(Minimum=0, Maximum=1))
     "Waveform [1]"=::NewEnumerator0
     "[1] Frequency"=5
     "[1] Amplitude Min/Max"=<0 1>
    )) {isSetParameters=true}
   &ScaleColor(
    Scale Alpha=&MergeHelper_FloatSwitch(
     0=&FloatFromCurve(
      FloatCurve=DataInterface() {propertyValues {Curve {
         defaultValue=1
         keys [
          {arriveTangent=5.5696, interpMode=RCIM_Cubic, leaveTangent=5.5696, tangentMode=RCTM_User, time=0, value=0}
          {arriveTangent=-0.8557, interpMode=RCIM_Cubic, leaveTangent=-0.8557, tangentMode=RCTM_User, time=0.1, value=1}
          {tangentMode=RCTM_User, time=1, value=0}
         ]
        }}}
      Scale Curve=&Lerp_Float(A=0.5, B=1.1, Alpha=$Particles.FlickerWaveform)
     )
     1=&FloatFromCurve(
      FloatCurve=DataInterface() {propertyValues {Curve {
         defaultValue=1
         keys [
          {interpMode=RCIM_Cubic, leaveTangentWeight=0.4012, tangentMode=RCTM_User, time=0, value=0}
          {arriveTangentWeight=0.0442, interpMode=RCIM_Cubic, leaveTangentWeight=0.3945, tangentMode=RCTM_User, time=0.1477, value=1}
          {arriveTangentWeight=0.1189, interpMode=RCIM_Cubic, leaveTangentWeight=0.1652, tangentMode=RCTM_User, time=0.5045, value=1}
          {arriveTangent=-2.3008, tangentMode=RCTM_User, time=1, value=0}
         ]
        }}}
      Scale Curve=&RandomRangeFloat(Minimum=0.5, Maximum=1)
     )
     2=?
     3=?
     Index=&SpawnGroup()
     Num=::NewEnumerator0
    )
    Scale RGB=&Multiply_VectorByFloat(
     Vector=&MergeHelper_VectorSwitch(
      0=&VectorFromFloat(Value=&FloatFromCurve(FloatCurve=DataInterface() {propertyValues {Curve {
           defaultValue=1
           keys [
            {arriveTangent=9.0725, interpMode=RCIM_Cubic, leaveTangent=9.0725, tangentMode=RCTM_User, time=0, value=1}
            {arriveTangent=-0.7324, interpMode=RCIM_Cubic, leaveTangent=-0.7324, tangentMode=RCTM_User, time=0.1363, value=1.8121}
            {arriveTangent=-0.0405, interpMode=RCIM_Cubic, leaveTangent=-0.0405, tangentMode=RCTM_User, time=1, value=1}
           ]
          }}}))
      1=&VectorFromFloat(Value=1)
      2=?
      3=?
      Index=&SpawnGroup()
      Num=::NewEnumerator0
     )
     Float=&Lerp_Float(A=1, B=1.25, Alpha=$Particles.FlickerWaveform)
    )
    ScaleA=true
    ScaleRGB=true
   )
   &ScaleSpriteSize(Scale Factor=&Vector2DFromFloat(Value=&MergeHelper_FloatSwitch(
      0=&FloatFromCurve(
       FloatCurve=DataInterface() {propertyValues {Curve {
          defaultValue=0
          keys [
           {arriveTangent=9.0725, interpMode=RCIM_Cubic, leaveTangent=9.0725, tangentMode=RCTM_User, time=0, value=0}
           {
            arriveTangent=-1.1669
            arriveTangentWeight=0.1274
            interpMode=RCIM_Cubic
            leaveTangent=-1.1669
            leaveTangentWeight=0.7778
            tangentMode=RCTM_User
            time=0.1408
            value=0.9847
           }
           {arriveTangent=-0.0405, interpMode=RCIM_Cubic, leaveTangent=-0.0405, tangentMode=RCTM_User, time=1, value=0}
          ]
         }}}
       Scale Curve=&Lerp_Float(A=0.5, B=1, Alpha=$Particles.FlickerWaveform)
      )
      1=&FloatFromCurve(
       FloatCurve=DataInterface() {propertyValues {Curve {
          defaultValue=0
          keys [
           {arriveTangent=5.1744, interpMode=RCIM_Cubic, leaveTangent=5.1744, tangentMode=RCTM_User, time=0, value=0}
           {arriveTangentWeight=0.0333, interpMode=RCIM_Cubic, leaveTangentWeight=0.2, tangentMode=RCTM_User, time=0.2136, value=1}
           {arriveTangentWeight=0.2001, interpMode=RCIM_Cubic, leaveTangentWeight=0.2809, tangentMode=RCTM_Break, time=0.2937, value=1}
           {arriveTangentWeight=0.2839, interpMode=RCIM_Cubic, tangentMode=RCTM_User, time=1, value=0}
          ]
         }}}
       Scale Curve=&Lerp_Float(A=0.5, B=1.2, Alpha=$Particles.FlickerWaveform)
      )
      2=?
      3=?
      Index=&SpawnGroup()
      Num=::NewEnumerator0
     ))) {module=ScaleSpriteSize001}
   &CurlNoiseForce(Noise Frequency=100, Noise Strength=200, Pan Noise=true, Pan Noise Field=<0 0 0.5>, Random Seed=10)
   &Drag(Drag=&RandomRangeFloat(Minimum=5, Maximum=6), Use Linear Drag=true, Use Rotational Drag=false)
   &SolveForcesAndVelocity(Clamp Velocity=false, Limit Acceleration=false)
   /SetVariables(
    Particles.Glow Color=&MergeHelper_ColorSwitch(
     0=&MakeLinearColorFromVectorAndFloat(
      "Vector (RGB)"=&MakeVectorFromLinearColorRGB(Color=&MultiplyLinearColor(Linear Color=$Particles.Color, Scale Factor=&MakeVector4_FromLinearColor(LinearColor=$Particles.Color)))
      "Float (Alpha)"=0.08
     )
     1=&MakeLinearColorFromVectorAndFloat(
      "Vector (RGB)"=&MakeVectorFromLinearColorRGB(Color=&MultiplyLinearColor(
        Linear Color=$Particles.Color
        Scale Factor=&MakeVector4_FromLinearColor(LinearColor=<rgba 0.15 0.0919 0.0339 1>)
       ))
      "Float (Alpha)"=1
     )
     2=?
     3=?
     Index=&SpawnGroup()
     Num=::NewEnumerator0
    )
    Particles.Glow Size=&Multiply_Vector2D(
     A=$Particles.SpriteSize
     B=&MergeHelper_Vector2Switch(0=<3.5 4.5>, 1=<0.5 0.5>, 2=?, 3=?, 4=?, Index=&SpawnGroup(), Num=::NewEnumerator0)
    )
   ) {isSetParameters=true}
  ]
  rendererClasses [&NiagaraSpriteRendererProperties]
  renderers [
   {rendererClass=&NiagaraSpriteRendererProperties}
   {rendererClass=&NiagaraSpriteRendererProperties, rendererIndex=1}
   {rendererClass=&NiagaraSpriteRendererProperties, rendererIndex=2}
   {rendererClass=&NiagaraSpriteRendererProperties, rendererIndex=3}
  ]
 }
 Trail_Instrument_Color001 {
  emitterUpdateScript [
   &EmitterState()
   &FN_SpawnAnimTrail_v2(AnimTrailWidthMode=::FromFirst)
  ]
  particleSpawnScript [
   /Game/Effects/Niagara/Modules/AnimTrail/FN_InitializeAnimTrail/FN_InitializeAnimTrailBoneIndecies(Second Bone or Socket Index=1)
   /Game/Effects/Niagara/Modules/AnimTrail/FN_InitializeAnimTrail/FN_InitializeAnimTrail(
    AnimTrailWidth=0.75
    Color=&MultiplyLinearColor(
     Linear Color=<rgba 1 0.987 0.9212 0.42>
     Scale Factor=&MakeVector4_FromLinearColor(LinearColor=$System.TOD Color)
    )
    Lifetime=0.15
   )
   &SystemLocation(MaskSpawnGroup=false, Offset Coordinate Space=::World)
   &AddVelocity(Velocity=<0 500 0>, Velocity Speed Scale=&RandomRangeFloat(Minimum=0.5, Maximum=1.2)) {enabled=false}
  ]
  particleUpdateScript [
   &ScaleRibbonWidth(Ribbon Width Scale=&FloatFromCurve(FloatCurve=DataInterface() {propertyValues {Curve {keys [
         {arriveTangent=-1.2579, interpMode=RCIM_Cubic, time=0, value=1}
         {arriveTangent=-0.9664, interpMode=RCIM_Cubic, leaveTangent=-0.9664, tangentMode=RCTM_User, time=1, value=0}
        ]}}}))
   &ScaleColor(
    Scale Alpha=&Multiply_Float(
     A=&FloatFromCurve(
      FloatCurve=DataInterface() {propertyValues {Curve {keys [
          {interpMode=RCIM_Cubic, leaveTangent=-1.2334, tangentMode=RCTM_Break, time=0, value=1}
          <at 0.3 0.1>
          {arriveTangent=-0.2139, interpMode=RCIM_Cubic, tangentMode=RCTM_Break, time=1, value=0}
         ]}}}
      Scale Curve=&FloatFromCurve(
       FloatCurve=DataInterface() {propertyValues {Curve {keys [<at 0 1>,<at 0.25 1>,<at 0.5 1>,<at 0.75 1>,<at 1 1>]}}}
       CurveIndex=Linked() {linkedVariable {name=NPC.FortniteNPC.TimeOfDay_0-1, type=NiagaraFloat}}
      )
     )
     B=&FN_FrontEndFloatModifier(Gameplay=1, Front End=0.7)
    )
    Scale RGB=&MakeVectorFromLinearColorRGB(Color=$System.TOD Color)
    ScaleA=true
    ScaleRGB=true
   ) {module=ScaleColor002}
   &FN_UpdateAnimTrail_v2()
   &SolveForcesAndVelocity(Clamp Velocity=false, Limit Acceleration=false) {enabled=false}
  ]
  rendererClasses [&NiagaraRibbonRendererProperties]
  renderers [{rendererClass=&NiagaraRibbonRendererProperties}]
 }
 Trail_Instrument_Color003 {
  emitterUpdateScript [
   &EmitterState()
   &FN_SpawnAnimTrail_v2()
  ]
  particleSpawnScript [
   &FN_InitializeAnimTrailBoneIndecies(First Bone or Socket Index=1)
   &FN_InitializeAnimTrail(
    AnimTrailWidth=1
    Color=&MultiplyLinearColor(
     Linear Color=<rgba 1 0.987 0.9212 0.42>
     Scale Factor=&MakeVector4_FromLinearColor(LinearColor=$System.TOD Color)
    )
    Lifetime=0.18
   )
   &SystemLocation(MaskSpawnGroup=false)
   &AddVelocity(Velocity=<0 500 0>) {enabled=false}
  ]
  particleUpdateScript [
   &ScaleRibbonWidth(Ribbon Width Scale=&FloatFromCurve(FloatCurve=DataInterface() {propertyValues {Curve {keys [
         {arriveTangent=-1.2579, interpMode=RCIM_Cubic, time=0, value=1}
         {arriveTangent=0.0472, interpMode=RCIM_Cubic, leaveTangent=0.0472, tangentMode=RCTM_User, time=1, value=0}
        ]}}}))
   &ScaleColor(
    Scale Alpha=&Multiply_Float(
     A=&FloatFromCurve(
      FloatCurve=DataInterface() {propertyValues {Curve {keys [
          {interpMode=RCIM_Cubic, leaveTangent=-0.7998, tangentMode=RCTM_Break, time=0, value=1}
          {arriveTangent=-0.3706, tangentMode=RCTM_User, time=0.3, value=0.2}
          {arriveTangent=-0.2139, interpMode=RCIM_Cubic, tangentMode=RCTM_Break, time=1, value=0}
         ]}}}
      Scale Curve=&FloatFromCurve(
       FloatCurve=DataInterface() {propertyValues {Curve {keys [<at 0 1>,<at 0.25 1>,<at 0.5 1>,<at 0.75 1>,<at 1 1>]}}}
       CurveIndex=Linked() {linkedVariable {name=NPC.FortniteNPC.TimeOfDay_0-1, type=NiagaraFloat}}
      )
     )
     B=&FN_FrontEndFloatModifier(Gameplay=1, Front End=0.7)
    )
    Scale RGB=&MakeVectorFromLinearColorRGB(Color=$System.TOD Color)
    ScaleA=true
    ScaleRGB=true
   ) {module=ScaleColor002}
   &FN_UpdateAnimTrail_v2()
   &SolveForcesAndVelocity(Clamp Velocity=false, Limit Acceleration=false) {enabled=false}
  ]
  rendererClasses [&NiagaraRibbonRendererProperties]
  renderers [{rendererClass=&NiagaraRibbonRendererProperties}]
 }
 WispTrail001 {
  emitterUpdateScript [
   &EmitterState(Life Cycle Mode=::NewEnumerator1, Loop Behavior=::NewEnumerator1, Loop Delay=0.005, Loop Duration=0.5, UseLoopDelay=true)
   &SpawnRate(SpawnRate=45, Use Spawn Probability=false)
  ]
  particleSpawnScript [
   &InitializeRibbon(
    Color=<rgba 1 0.9626 0.8613 0.6>
    Lifetime=0.2
    Module.Write Color=true
    Module.Write Lifetime=true
    Module.Write Mass=false
    Module.Write Material Random=true
    Module.Write Position=false
    Module.Write RibbonTwist=false
    Module.Write RibbonWidth=true
    Ribbon Width=30
   )
   &SocketLocation(
    Interpolate Socket Sampling=true
    SamplePosOffset=false
    Socket Index=1
    Socket Reader=$System.Socket Reader
    Socket Sampling Mode=::NewEnumerator3
    VelocitySpeedLimit=false
   ) {enabled=false}
   &SkeletalMeshLocation(
    MaskSpawnGroup=false
    Mesh Sampling Type=::NewEnumerator5
    Sampled Position Offset=<0 0 4>
    Skeletal Mesh=$User.SkeletalMesh
    Socket Sampling Mode=::NewEnumerator5
   )
   &AddVelocity(Velocity=<0 500 0>) {enabled=false}
  ]
  particleUpdateScript [
   &ParticleState()
   &CurlNoiseForce(
    Bias Noise=false
    Noise Frequency=300
    Noise Strength=&FloatFromCurve(
     FloatCurve=DataInterface() {propertyValues {Curve {
        defaultValue=1
        keys [
         {arriveTangent=4.1492, interpMode=RCIM_Cubic, leaveTangent=4.1492, tangentMode=RCTM_User, time=0.5, value=0}
         {interpMode=RCIM_Cubic, tangentMode=RCTM_User, time=0.9, value=1}
        ]
       }}}
     Scale Curve=3000
    )
    Pan Noise=false
   )
   &SolveForcesAndVelocity(Clamp Velocity=false, Limit Acceleration=false)
   &ScaleColor(
    Scale Alpha=&Multiply_Float(
     A=&FloatFromCurve(
      FloatCurve=DataInterface() {propertyValues {Curve {keys [
          {interpMode=RCIM_Cubic, leaveTangent=-3.1628, leaveTangentWeight=0.4326, tangentMode=RCTM_Break, time=0.2, value=1}
          {arriveTangent=-1.2479, arriveTangentWeight=0.2986, interpMode=RCIM_Cubic, leaveTangent=-1.2479, tangentMode=RCTM_User, time=1, value=0}
         ]}}}
      Scale Curve=&FloatFromCurve(
       FloatCurve=DataInterface() {propertyValues {Curve {keys [<at 0 1>,<at 0.497 1 RCIM_Cubic>,<at 1 1>]}}}
       CurveIndex=Linked() {linkedVariable {name=NPC.FortniteNPC.TimeOfDay_0-1, type=NiagaraFloat}}
      )
     )
     B=&MakeCustomFloatFromBool(Bool=$User.bIsFrontEndPreview, True Float=0.5, False Float=1)
    )
    Scale RGB=&Multiply_Vector(
     A=&MakeVectorFromLinearColorRGB(Color=$System.TOD Color)
     B=&MakeVectorFromLinearColorRGB(Color=&ColorFromCurve(ColorCurve=DataInterface() {propertyValues {
         AlphaCurve {keys [<at 0 1 RCIM_Cubic>]}
         BlueCurve {keys [<at 0.0901 1 RCIM_Cubic>, {arriveTangent=-0.697, interpMode=RCIM_Cubic, leaveTangent=-0.697, time=0.6719, value=0.7542}, <at 0.991 0.3721>]}
         GreenCurve {keys [<at 0.0901 1 RCIM_Cubic>, {arriveTangent=-0.5019, interpMode=RCIM_Cubic, leaveTangent=-0.5019, time=0.6719, value=0.985}, <at 0.991 0.5478>]}
         RedCurve {keys [<at 0.0901 1 RCIM_Cubic>, {arriveTangent=-0.4856, interpMode=RCIM_Cubic, leaveTangent=-0.4856, time=0.6719, value=1}, <at 0.991 0.5625>]}
        }}))
    )
    ScaleA=true
    ScaleRGB=true
   ) {module=ScaleColor002}
   &ScaleRibbonWidth(Ribbon Width Scale=&FloatFromCurve(FloatCurve=DataInterface() {propertyValues {Curve {keys [
         {arriveTangent=2.6972, interpMode=RCIM_Cubic, leaveTangent=2.6972, leaveTangentWeight=0.2293, tangentMode=RCTM_User, time=0.1, value=1}
         {arriveTangent=4.8659, interpMode=RCIM_Cubic, leaveTangent=4.8659, tangentMode=RCTM_User, time=0.75, value=2.75}
         {arriveTangent=24.818, interpMode=RCIM_Cubic, leaveTangent=24.818, tangentMode=RCTM_User, time=1, value=6}
        ]}}}))
   &DynamicMaterialParameters(
    Index 0 Param 1=&FloatFromCurve(
     FloatCurve=DataInterface() {propertyValues {Curve {
        defaultValue=0
        keys [
         {arriveTangent=1.5964, interpMode=RCIM_Cubic, leaveTangent=1.5964, tangentMode=RCTM_User, time=0, value=0}
         {arriveTangent=0.7982, interpMode=RCIM_Cubic, leaveTangent=0.7982, tangentMode=RCTM_User, time=1, value=1}
        ]
       }}}
     Scale Curve=0.75
    )
    Index 0 Param 3=-0.8
    Index 0 Param 4=&FloatFromCurve(FloatCurve=DataInterface() {propertyValues {Curve {keys [
         {interpMode=RCIM_Cubic, tangentMode=RCTM_User, time=0, value=1}
         {arriveTangent=-0.6219, interpMode=RCIM_Cubic, leaveTangent=-0.6219, tangentMode=RCTM_User, time=1, value=0}
        ]}}})
    Module.Param0WriteEnabled=true
    Module.Param1WriteEnabled=true
    Module.Param2WriteEnabled=true
    Module.Param3WriteEnabled=true
   )
  ]
  rendererClasses [&NiagaraRibbonRendererProperties]
  renderers [{rendererClass=&NiagaraRibbonRendererProperties}]
 }
}
summary {
 systemName=NS_Pickaxe_Guitar_IndoorLace_SwingTrail
 emitters [
  {emitterName=Trail_Instrument_Color001, rendererClasses [&NiagaraRibbonRendererProperties]}
  {emitterName=WispTrail001, rendererClasses [&NiagaraRibbonRendererProperties]}
  {emitterName=Trail_Instrument_Color003, rendererClasses [&NiagaraRibbonRendererProperties]}
  {emitterName=Flares_Sparkles, rendererClasses [&NiagaraSpriteRendererProperties]}
 ]
 userVariables [
  User.bIsFrontEnd {type=NiagaraBool}
  User.bIsFrontEndPreview {type=NiagaraBool}
  User.RibbonIDOffset {type=NiagaraInt32}
  User.SkeletalMesh {
   type=NiagaraDataInterfaceSkeletalMesh
   defaultValue {
    struct=NiagaraExt_VariableValue_DataInterface
    value {dataInterface=&self:NiagaraDataInterfaceSkeletalMesh_0}
   }
  }
 ]
}
